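# FM_BandPortal is a Sound Cue for band portal audio in SparksMaps, assigned to the Fort_SFX_InGame sound class.
# A Mixer root passes a single Wave Player of the band_portal1 wave through at full volume.
# The mixer's second input is empty, so its reduced volume setting has no source to act on.
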SoundCue /SparksMaps/Audio/FM_BandPortal
root=Mixer_0
nodes [
 {id=Mixer_0, children [WavePlayer_0,null], properties {inputVolume [1,0.2]}}
 {id=WavePlayer_0, properties {soundWaveAssetPtr=/SparksMaps/Audio/Wavs/band_portal1}}
]
properties {maxDistance=2097152, soundClassObject=&Fort_SFX_InGame, volumeMultiplier=1}
