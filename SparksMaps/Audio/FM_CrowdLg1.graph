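# Sound cue for a looping large, close crowd ambience in the SparksMaps audio set.
# A single looping wave player feeds the cue output directly, with no randomization or mixing.
# It uses a shared large-crowd attenuation asset and the in-game SFX sound class.
# A concurrency override limits it to one playing instance at a time.
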SoundCue /SparksMaps/Audio/FM_CrowdLg1
root=WavePlayer_1
nodes [{id=WavePlayer_1, properties {bLooping=true, soundWaveAssetPtr=/SparksMaps/Audio/Wavs/crowd_lg_close1}}]
properties {
 attenuationSettings=&/FM_Attenu_Crowd_Lg
 bOverrideConcurrency=true
 maxDistance=10000
 soundClassObject=&Fort_SFX_InGame
 concurrencyOverrides {maxCount=1}
}
