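# FM_CrowdLg2 is a SparksMaps SoundCue, named as a large-crowd sound, that has no nodes or root and so plays no audio as it stands.
# It keeps only playback settings: the large-crowd attenuation asset and the in-game SFX sound class.
# A concurrency override limits the cue to a single playing instance.
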SoundCue /SparksMaps/Audio/FM_CrowdLg2
root=null
nodes []
properties {
 attenuationSettings=&/FM_Attenu_Crowd_Lg
 bOverrideConcurrency=true
 maxDistance=10000
 soundClassObject=&Fort_SFX_InGame
 concurrencyOverrides {maxCount=1}
}
