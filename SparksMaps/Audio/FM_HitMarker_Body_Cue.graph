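# Sound cue for a body-shot hit marker in SparksMaps, routed to a 2D hit-notification sound class at low overall volume.
# A mixer layers a pitched-up, slightly reduced VoteGo wave over a random pick of three Athena HitMarker_Body waves, each branch passing through a modulator for variation.
# A two-voice, stop-oldest concurrency set with volume scaling, as its name indicates, keeps rapid hits from piling up.
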SoundCue /SparksMaps/Audio/FM_HitMarker_Body_Cue
root=Mixer_0
nodes [
 {id=Mixer_0, children [Modulator_1,Modulator_0], properties {inputVolume [0.75,1]}}
 {id=Modulator_0, children [Random_0]}
 {id=Modulator_1, children [WavePlayer_3], properties {pitchMax=1.52, pitchMin=1.47}}
 {id=Random_0, children [WavePlayer_0,WavePlayer_1,WavePlayer_2], properties {bShouldExcludeFromBranchCulling=true, weights [1,1,1]}}
 {id=WavePlayer_0, properties {soundWaveAssetPtr=/Game/Athena/Sounds/HitNotifications/HitMarker_Body_01}}
 {id=WavePlayer_1, properties {soundWaveAssetPtr=/Game/Athena/Sounds/HitNotifications/HitMarker_Body_02}}
 {id=WavePlayer_2, properties {soundWaveAssetPtr=/Game/Athena/Sounds/HitNotifications/HitMarker_Body_03}}
 {id=WavePlayer_3, properties {soundWaveAssetPtr=/SparksMaps/Audio/Wavs/VoteGo}}
]
properties {
 maxDistance=2097152
 soundClassObject=&Fort_2DHitNotifications_Instigate_PSM
 volumeMultiplier=0.05
 concurrencyOverrides {maxCount=2, resolutionRule=StopOldest}
 concurrencySet [&HitNotifications_TwoMaximum_StopOldest_VolumeScale]
}
