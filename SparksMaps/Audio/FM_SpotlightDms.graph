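# Sound cue for a drums spotlight moment in SparksMaps, layering a drum wave with a crowd song-end reaction.
# A Mixer plays the spotlight drums at full input volume alongside a much quieter Random node that picks one of three equally weighted large-crowd waves.
# The cue routes to the in-game SFX sound class, with slightly reduced volume and a very large max distance.
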SoundCue /SparksMaps/Audio/FM_SpotlightDms
root=Mixer_0
nodes [
 {id=Mixer_0, children [WavePlayer_0,Random_0], properties {inputVolume [1,0.2]}}
 {id=Random_0, children [WavePlayer_5,WavePlayer_6,WavePlayer_7], properties {weights [1,1,1]}}
 {id=WavePlayer_0, properties {soundWaveAssetPtr=/SparksMaps/Audio/Wavs/spotlight_drums1}}
 {id=WavePlayer_5, properties {soundWaveAssetPtr=/SparksMaps/Audio/Wavs/crowd_lg_song_end3}}
 {id=WavePlayer_6, properties {soundWaveAssetPtr=/SparksMaps/Audio/Wavs/crowd_lg_song_end1}}
 {id=WavePlayer_7, properties {soundWaveAssetPtr=/SparksMaps/Audio/Wavs/crowd_lg_song_end2}}
]
properties {maxDistance=2097152, soundClassObject=&Fort_SFX_InGame, volumeMultiplier=0.9}
