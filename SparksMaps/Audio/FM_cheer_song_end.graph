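# A SoundCue that plays a large crowd cheer for the end of a song in SparksMaps.
# A Random node picks one of three equally weighted crowd wave variations on each play.
# It overrides attenuation to be non-spatialized with distance falloff and caps concurrency at two instances, playing at reduced volume in the in-game SFX sound class.
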SoundCue /SparksMaps/Audio/FM_cheer_song_end
root=Random_0
nodes [
 {id=Random_0, children [WavePlayer_3,WavePlayer_4,WavePlayer_5], properties {weights [1,1,1]}}
 {id=WavePlayer_3, properties {soundWaveAssetPtr=/SparksMaps/Audio/Wavs/crowd_lg_song_end1}}
 {id=WavePlayer_4, properties {soundWaveAssetPtr=/SparksMaps/Audio/Wavs/crowd_lg_song_end2}}
 {id=WavePlayer_5, properties {soundWaveAssetPtr=/SparksMaps/Audio/Wavs/crowd_lg_song_end3}}
]
properties {
 bOverrideAttenuation=true
 bOverrideConcurrency=true
 maxDistance=11000
 soundClassObject=&Fort_SFX_InGame
 volumeMultiplier=0.6
 attenuationOverrides {bSpatialize=false, falloffDistance=6000, attenuationShapeExtents {x=5000}}
 concurrencyOverrides {maxCount=2}
}
