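# Sound cue for a large crowd cheer at the start of a song, routed through the in-game SFX sound class.
# A short randomized delay feeds a Random node that picks one of three equally weighted crowd wave variations.
# Attenuation and concurrency overrides make it non-spatialized with distance falloff and cap simultaneous instances at a low count.
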SoundCue /SparksMaps/Audio/FM_cheer_song_start
root=Delay_0
nodes [
 {id=Delay_0, children [Random_0], properties {delayMax=2.5, delayMin=2}}
 {id=Random_0, children [WavePlayer_0,WavePlayer_1,WavePlayer_2], properties {weights [1,1,1]}}
 {id=WavePlayer_0, properties {soundWaveAssetPtr=/SparksMaps/Audio/Wavs/crowd_lg_song_start3}}
 {id=WavePlayer_1, properties {soundWaveAssetPtr=/SparksMaps/Audio/Wavs/crowd_lg_song_start1}}
 {id=WavePlayer_2, properties {soundWaveAssetPtr=/SparksMaps/Audio/Wavs/crowd_lg_song_start2}}
]
properties {
 bOverrideAttenuation=true
 bOverrideConcurrency=true
 maxDistance=11000
 soundClassObject=&Fort_SFX_InGame
 volumeMultiplier=0.6
 attenuationOverrides {bSpatialize=false, falloffDistance=6000, attenuationShapeExtents {x=5000}}
 concurrencyOverrides {maxCount=2}
}
