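# FM_hum_mach is a SoundCue that plays a looping ambient machinery hum in the SparksMaps audio set.
# A weighted Random node picks one of four looping waves (two dehumidifier hums, a furnace and an oven hood), slightly favoring the first dehumidifier.
# Output plays at reduced volume on the in-game SFX class, with FM_MachHum_att attenuation assigned alongside natural-sound falloff override values.
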SoundCue /SparksMaps/Audio/FM_hum_mach
root=Random_0
nodes [
 {id=Random_0, children [WavePlayer_0,WavePlayer_1,WavePlayer_2,WavePlayer_3], properties {weights [1,0.8,0.8,0.8]}}
 {id=WavePlayer_0, properties {bLooping=true, soundWaveAssetPtr=/SparksMaps/Audio/Wavs/amb_hum_dehumid1}}
 {id=WavePlayer_1, properties {bLooping=true, soundWaveAssetPtr=/SparksMaps/Audio/Wavs/amb_hum_dehumid2}}
 {id=WavePlayer_2, properties {bLooping=true, soundWaveAssetPtr=/SparksMaps/Audio/Wavs/amb_hum_furnace}}
 {id=WavePlayer_3, properties {bLooping=true, soundWaveAssetPtr=/SparksMaps/Audio/Wavs/amb_hum_ovenhood}}
]
properties {
 attenuationSettings=&/FM_MachHum_att
 maxDistance=3300
 soundClassObject=&Fort_SFX_InGame
 volumeMultiplier=0.3
 attenuationOverrides {distanceAlgorithm=NaturalSound, falloffDistance=2000, attenuationShapeExtents {x=300}}
}
