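# FM_od_deploy is a Sound Cue that plays an "od_deploy" sound effect, judging by its name a deploy sound.
# A Random node picks one of three wave variations (od_deploy_08 to 10) with equal weight.
# The cue routes to the Fort_SFX_InGame sound class, with reduced volume and a very large max distance.
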SoundCue /SparksMaps/Audio/FM_od_deploy
root=Random_0
nodes [
 {id=Random_0, children [WavePlayer_0,WavePlayer_1,WavePlayer_2], properties {weights [1,1,1]}}
 {id=WavePlayer_0, properties {soundWaveAssetPtr=/SparksMaps/Audio/Wavs/od_deploy_08}}
 {id=WavePlayer_1, properties {soundWaveAssetPtr=/SparksMaps/Audio/Wavs/od_deploy_09}}
 {id=WavePlayer_2, properties {soundWaveAssetPtr=/SparksMaps/Audio/Wavs/od_deploy_10}}
]
properties {maxDistance=2097152, soundClassObject=&Fort_SFX_InGame, volumeMultiplier=0.6}
