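# SoundCue for a bandmate overdrive deploy sound, judging by its name, in the SparksMaps audio content.
# A single Wave Player plays the od_deploy_band_01 wave through a Modulator that randomizes pitch slightly upward while keeping volume fixed.
# It routes to the in-game SFX sound class.
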SoundCue /SparksMaps/Audio/FM_od_deploy_bandmate
root=Modulator_1
nodes [
 {id=Modulator_1, children [WavePlayer_3], properties {pitchMax=1.2, pitchMin=1, volumeMax=1, volumeMin=1}}
 {id=WavePlayer_3, properties {soundWaveAssetPtr=/SparksMaps/Audio/Wavs/od_deploy_band_01}}
]
properties {maxDistance=2097152, soundClassObject=&Fort_SFX_InGame}
