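# Sound cue for a looping deploy buzz effect, likely an "od" deploy sound in the SparksMaps audio set.
# A single looping wave player feeds the root directly with the od_deploy_buzz_lp wave, with no modulation or attenuation nodes.
# It plays in the in-game SFX sound class at reduced volume, with a very large max distance.
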SoundCue /SparksMaps/Audio/FM_od_deploybuzz
root=WavePlayer_0
nodes [{id=WavePlayer_0, properties {bLooping=true, soundWaveAssetPtr=/SparksMaps/Audio/Wavs/od_deploy_buzz_lp}}]
properties {maxDistance=2097152, soundClassObject=&Fort_SFX_InGame, volumeMultiplier=0.4}
