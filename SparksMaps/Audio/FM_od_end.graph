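# FM_od_end is a SoundCue in SparksMaps audio that plays a single od_end sound wave.
# One Wave Player feeds the output directly, with no randomization, modulation or mixing.
# The cue routes to the Fort_SFX_InGame sound class at reduced volume and sets a very large max distance.
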SoundCue /SparksMaps/Audio/FM_od_end
root=WavePlayer_0
nodes [{id=WavePlayer_0, properties {soundWaveAssetPtr=/SparksMaps/Audio/Wavs/od_end_06}}]
properties {maxDistance=2097152, soundClassObject=&Fort_SFX_InGame, volumeMultiplier=0.5}
