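# FM_od_ready is a Sound Cue in SparksMaps that plays an "od ready" audio cue, likely signalling an overdrive-ready state.
# A single Wave Player feeds the root directly with the od_ready_02 wave, with no randomization or modulation.
# It routes through the Fort_SFX_InGame sound class at reduced volume, with a very large max distance.
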SoundCue /SparksMaps/Audio/FM_od_ready
root=WavePlayer_1
nodes [{id=WavePlayer_1, properties {soundWaveAssetPtr=/SparksMaps/Audio/Wavs/od_ready_02}}]
properties {maxDistance=2097152, soundClassObject=&Fort_SFX_InGame, volumeMultiplier=0.6}
