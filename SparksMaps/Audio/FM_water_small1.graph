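# FM_water_small1 is a Sound Cue for a small looping water sound in SparksMaps audio, routed to the Fort_SFX_InGame sound class at reduced volume.
# A single looping Wave Player feeds the output directly with the water_small1 wave.
# It references the FM_Attenu_Crowd_Med attenuation asset, but attenuation override is enabled, so the cue's own falloff settings apply instead.
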SoundCue /SparksMaps/Audio/FM_water_small1
root=WavePlayer_4
nodes [{id=WavePlayer_4, properties {bLooping=true, soundWaveAssetPtr=/SparksMaps/Audio/Wavs/water_small1}}]
properties {
 attenuationSettings=&/FM_Attenu_Crowd_Med
 bOverrideAttenuation=true
 maxDistance=1900
 soundClassObject=&Fort_SFX_InGame
 volumeMultiplier=0.5
 attenuationOverrides {falloffDistance=1500}
}
