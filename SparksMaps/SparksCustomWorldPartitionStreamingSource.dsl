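; Sparks actor that switches a custom World Partition streaming source on or off according to the current playlist.
; On BeginPlay it waits asynchronously for playlist data, then enables its StreamingSource only if Sparks custom streaming is enabled and the playlist shares any configured tag, with an empty tag list matching every playlist.
; Otherwise it disables the source.
; Either way it sets bHasProcessedStreamingSources and prints the color-coded result to screen.
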(event EventBeginPlay
  (bind _returnvalue (Class|SparksControllerComponentDisablePCStreamingSource|IsSparksCustomStreamingEnabled))
  (bind _playlistgameplaytags (Variables|Default|GetPlaylistGameplayTags))
  (bind _is_not_empty (Utilities|GameplayTagContainerNotEmpty _playlistgameplaytags))
  (bind _returnvalue_2 (not _is_not_empty))
  (bind _playlistgameplaytags_2 (Variables|Default|GetPlaylistGameplayTags))
  (bind _streamingsource (Variables|Default|GetStreamingSource))
  (|CallParentFunction)
  (bind (_gamestate _playlist) (Class|FortAsyncActionCurrentPlaylistReady|PlaylistDataReadyAsync)
    (:OnReady
      (bind _asfort_playlist_athena (Utilities|Casting|CastToFortPlaylistAthena _playlist))
      (if (and (or (GameplayTags|HasAnyTags (Variables|/Script/FortniteGame.FortPlaylist|GetGameplayTagContainer _asfort_playlist_athena) _playlistgameplaytags_2) _returnvalue_2) _returnvalue)
        (Streaming|EnableStreamingSource _streamingsource)
        (else
          (Streaming|DisableStreamingSource _streamingsource)))
      (Variables|Default|SetbHasProcessedStreamingSources true)
      (Development|PrintString (Utilities|String|Append "Sparks Custom Streaming Source " (Utilities|GetDisplayName self) " is " (Utilities|String|SelectString "Enabled!" "Disabled!" (and (or (GameplayTags|HasAnyTags (Variables|/Script/FortniteGame.FortPlaylist|GetGameplayTagContainer _asfort_playlist_athena) _playlistgameplaytags_2) _returnvalue_2) _returnvalue))) true false (Math|Color|SelectColor "(R=0.000000,G=0.454000,B=0.000000,A=1.000000)" "(R=1.000000,G=0.217016,B=0.648386,A=1.000000)" (and (or (GameplayTags|HasAnyTags (Variables|/Script/FortniteGame.FortPlaylist|GetGameplayTagContainer _asfort_playlist_athena) _playlistgameplaytags_2) _returnvalue_2) _returnvalue)) 15.0))))

(event Collision|EventActorBeginOverlap
  (params (OtherActor :type Actor%20Object%20Reference)))

(event EventTick
  (params (DeltaSeconds :type Float%20%28single-precision%29)))

(fn ConstructionScript)
